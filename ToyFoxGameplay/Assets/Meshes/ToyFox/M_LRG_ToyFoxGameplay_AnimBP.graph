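# ToyFox gameplay Animation Blueprint that copies its attached parent mesh's pose and layers shoulder-cannon animation, aiming and first-person adjustments onto it.
# An idle/fire state machine with additive shake blends in from the backpack bone, then feeds a Control Rig and per-cannon aim rotations.
# First-person view instead offsets and scales the backpack, spreads the cannons and layers a cannon AO frame from neck_01, before the default slot.
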AnimBlueprint /ToyFoxGameplay/Assets/Meshes/ToyFox/M_LRG_ToyFoxGameplay_AnimBP
target_skeleton=&/M_LRG_ToyFox_Skeleton
graphs [
 AnimGraph {nodes [
   AnimGraphNode_StateMachine_0 = cannons_SM() {is=title AnimGraphNode_StateMachine}
   n1 = Get CannonShake() -> (CannonShake:Rotator)
   n2 = Get ShakeAlpha() -> (ShakeAlpha:double)
   n3 = "Transform (Modify) Bone - Bone: dyn_cannon_l"(
    ComponentPose="Transform (Modify) Bone - Bone: dyn_cannon_r"(ComponentPose=Local To Component(LocalPose=@AnimGraphNode_StateMachine_0), Alpha=@n2, Translation="0,0,0", Rotation=@n1, Scale="1,1,1") {rotationMode=BMM_Additive, alphaBoolBlend {blendOption=Linear}, boneToModify {boneName=dyn_cannon_r}}
    Alpha=@n2
    Translation="0,0,0"
    Rotation=@n1
    Scale="1,1,1"
   ) {rotationMode=BMM_Additive, alphaBoolBlend {blendOption=Linear}, boneToModify {boneName=dyn_cannon_l}}
   n4 = Control Rig!(
    Source=Layered blend per bone(BasePose=Use cached pose 'fullbodyCache'(), BlendPoses_0=Component To Local(ComponentPose=@n3)) {layerSetup [{branchFilters [{boneName=dyn_backpack}]}]}
    Alpha
    TurnOffSkirt:double=Get TurnOffSkirt() -> (TurnOffSkirt:float)
   ) -> (Pose) {
    alphaBoolBlend {blendOption=Linear}
    controlRigAssetReference {blueprintRigClass=&/M_LRG_ToyFox_CtrlRig.M_LRG_ToyFox_CtrlRig_C}
   }
   n5 = Get!() -> (AimV2Right_X:double, AimV2Right_Y:double)
   n6 = "Transform (Modify) Bone - Bone: dyn_cannon_r"!(
    ComponentPose=Local To Component(LocalPose=@n4)
    Alpha=1
    Translation="0,0,0"
    Rotation_Roll:float=@n5
    Rotation_Pitch:float=0
    Rotation_Yaw:float=@n5.AimV2Right_Y
    Scale="1,1,1"
   ) -> (Pose) {rotationMode=BMM_Replace, rotationSpace=BCS_BoneSpace, alphaBoolBlend {blendOption=Linear}, boneToModify {boneName=dyn_cannon_r}}
   n7 = Get!() -> (AimV2Left_X:double, AimV2Left_Y:double)
   n8 = "Transform (Modify) Bone - Bone: dyn_cannon_l"!(
    ComponentPose=@n6
    Alpha=1
    Translation="0,0,0"
    Rotation_Roll:float=@n7
    Rotation_Pitch:float=0
    Rotation_Yaw:float=@n7.AimV2Left_Y
    Scale="1,1,1"
   ) -> (Pose) {rotationMode=BMM_Replace, rotationSpace=BCS_BoneSpace, alphaBoolBlend {blendOption=Linear}, boneToModify {boneName=dyn_cannon_l}}
   n9 = Component To Local(ComponentPose=@n8)
   n10 = "Transform (Modify) Bone - Bone: dyn_backpack"(
    ComponentPose=Local To Component(LocalPose=Use cached pose 'fullbodyCache'())
    Alpha=1
    Translation=Get FirstPCannonOffset() -> (FirstPCannonOffset:Vector)
    Rotation="0,0,0"
    Scale=Get FirstPCannonScale() -> (FirstPCannonScale:Vector)
   ) {
    rotationMode=BMM_Replace
    rotationSpace=BCS_BoneSpace
    scaleMode=BMM_Replace
    translationMode=BMM_Additive
    translationSpace=BCS_ParentBoneSpace
    alphaBoolBlend {blendOption=Linear}
    boneToModify {boneName=dyn_backpack}
   }
   n11 = "Transform (Modify) Bone - Bone: dyn_cannon_l"!(
    ComponentPose=@n10
    Alpha=1
    Translation=Get NegativeFirstPCannonSpread() -> (NegativeFirstPCannonSpread:Vector)
    Rotation:Rotator="0,0,0"
    Scale="1,1,1"
   ) -> (Pose) {
    rotationMode=BMM_Replace
    rotationSpace=BCS_BoneSpace
    scaleMode=BMM_Replace
    translationMode=BMM_Additive
    translationSpace=BCS_ParentBoneSpace
    alphaBoolBlend {blendOption=Linear}
    boneToModify {boneName=dyn_cannon_l}
   }
   n12 = "Transform (Modify) Bone - Bone: dyn_cannon_r"!(ComponentPose=@n11, Alpha=1, Translation=Get FirstPCannonSpread() -> (FirstPCannonSpread:Vector), Rotation:Rotator="0,0,0", Scale="1,1,1") -> (Pose) {
    rotationMode=BMM_Replace
    rotationSpace=BCS_BoneSpace
    scaleMode=BMM_Replace
    translationMode=BMM_Additive
    translationSpace=BCS_ParentBoneSpace
    alphaBoolBlend {blendOption=Linear}
    boneToModify {boneName=dyn_cannon_r}
   }
   n13 = Layered blend per bone(
    BasePose=Component To Local(ComponentPose=@n12)
    BlendPoses_0="Sequence Evaluator (by time) 'ToyFox_Cannon_C_AO'"!(ExplicitTime:float=0) -> (Pose:Pose Link Structure) {is=title AnimGraphNode_SequenceEvaluator, sequence=/ToyFoxGameplay/Assets/Animations/ToyFox_Cannon_C_AO}
   ) {layerSetup [{branchFilters [{boneName=neck_01}]}]}
   n14 = Get bIsFirstPersonView() -> (bIsFirstPersonView:bool)
   Output Pose(Result=Slot 'DefaultSlot'(Source=Blend Poses by bool(BlendPose_0=@n13, BlendPose_1=@n9, bActiveValue=@n14) {blendType=Linear}))
   Save cached pose 'fullbodyCache'(Pose=Copy Pose From Mesh() {bUseAttachedParent=true}) {cacheName=fullbodyCache}
  ]}
 cannons_SM {
  entry=idle
  path=AnimGraph.AnimGraphNode_StateMachine_0.cannons_SM
  ids=id
  nodes [
   AnimStateNode_0 = idle() {is=title AnimStateNode}
   AnimStateNode_1 = fire() {is=title AnimStateNode}
   AnimStateTransitionNode_0 = AnimStateTransitionNode() {crossfadeDuration=0}
   AnimStateTransitionNode_1 = AnimStateTransitionNode() {bAutomaticRuleBasedOnSequencePlayerInState=true, crossfadeDuration=0}
  ]
  links [
   {
    dst=idle
    rule=AnimGraph.AnimGraphNode_StateMachine_0.cannons_SM.AnimStateTransitionNode_1.Transition
    src=fire
    via=AnimStateTransitionNode_1
   }
   {
    dst=fire
    rule=AnimGraph.AnimGraphNode_StateMachine_0.cannons_SM.AnimStateTransitionNode_0.Transition
    src=idle
    via=AnimStateTransitionNode_0
   }
  ]
 }
 idle {
  path=AnimGraph.AnimGraphNode_StateMachine_0.cannons_SM.AnimStateNode_0.idle
  nodes [Output Animation Pose(Result=Sequence Player 'ToyFox_LRG_Cannon_Idle'() {bLoopAnimation=false, sequence=/ToyFoxGameplay/Assets/Animations/ToyFox_LRG_Cannon_Idle})]
 }
 fire {
  path=AnimGraph.AnimGraphNode_StateMachine_0.cannons_SM.AnimStateNode_1.fire
  nodes [Output Animation Pose(Result=Blend Poses by bool(
     BlendPose_0=Sequence Player 'ToyFox_LRG_Cannon_FireTEST'() {bLoopAnimation=false, playRate=0.5, sequence=/ToyFoxGameplay/Assets/Animations/ToyFox_LRG_Cannon_FireTEST}
     BlendPose_1=Sequence Player 'ToyFox_LRG_Cannon_Fire'() {bLoopAnimation=false, playRate=0.3, sequence=/ToyFoxGameplay/Assets/Animations/ToyFox_LRG_Cannon_Fire, startPosition=0.07}
     bActiveValue=Get bUseKickbackAnim() -> (bUseKickbackAnim:bool)
    ) {blendType=Linear})]
 }
 Transition {
  path=AnimGraph.AnimGraphNode_StateMachine_0.cannons_SM.AnimStateTransitionNode_0.Transition
  nodes [Result(bCanEnterTransition=Get bAbilityCannonFire() -> (bAbilityCannonFire:bool))]
 }
 Transition {path=AnimGraph.AnimGraphNode_StateMachine_0.cannons_SM.AnimStateTransitionNode_1.Transition, nodes [Result()]}
 EventGraph {links [], nodes []}
]
variables []
