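# Physics asset for the ToyFox skeletal mesh, defining its per-bone collision bodies and joints.
# Capsules and boxes cover the spine, clavicles, head, calves and feet, while upper arm, forearm, neck and thigh bodies have no shapes.
# Constraints chain spine, arms and head together, but each leg forms a separate chain with no link to the spine.
# All constraint swing and twist motions are Free, so their angle limits are inactive.
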PhysicsAsset /ToyFoxGameplay/Assets/Meshes/ToyFox/M_LRG_ToyFox_PhysicsAsset
bodies {
 calf_l [
  {
   shapeType=Box
   extentX=16.5
   extentY=25.5
   extentZ=21.5
   shapeName=calf_l_box
   center=<-1.9068 -0.982 1.6338>
   rotation=<rot 3.558 -0.0112 0.2967>
  }
  {
   shapeType=Box
   extentX=39.5
   extentY=22.5
   extentZ=21.5
   shapeName=calf_l_box
   center=<-29.6891 -2.0548 -0.1049>
   rotation=<rot 2.8237 -0.0002 0>
  }
 ]
 calf_r [
  {
   shapeType=Box
   extentX=16.5
   extentY=25.5
   extentZ=21.5
   shapeName=calf_r_box
   center=<1.9068 0.982 -1.6338>
   rotation=<rot 3.558 -0.0112 -179.7033>
  }
  {
   shapeType=Box
   extentX=43.3125
   extentY=24.375
   extentZ=21.5
   shapeName=calf_r_box
   center=<31.4986 2.0547 0.1941>
   rotation=<rot 2.8237 -0.0002 -180>
  }
 ]
 clavicle_l [{shapeType=Capsule, length=3.5623, radius=10.5194, center=<3.9566 -1.0862 -2.1127>, rotation=<rot -77.5566 177.3915 7.7314>}]
 clavicle_r [{shapeType=Capsule, length=3.5623, radius=10.5194, center=<-3.9566 1.0862 2.1127>, rotation=<rot -77.5566 177.3915 -172.2686>}]
 foot_l [{shapeType=Capsule, length=17.105, radius=10.4344, center=<-4.3017 10.6873 0.207>, rotation=<rot 0.0387 1.4695 -88.6654>}]
 foot_r [{shapeType=Capsule, length=17.6042, radius=11.5003, center=<4.304 -10.7729 -0.409>, rotation=<rot 0.0387 1.4698 86.1431>}]
 head [
  {shapeType=Capsule, length=10.362, radius=10.3796, center=<6.3613 -2.5134 0>, rotation=<rot 0 -90 0>}
  {shapeType=Capsule, length=24.4731, radius=3.345, shapeName=head_capsule, center=<9.0706 -5.1144 23.2899>, rotation=<rot 0 -0.019 90.2065>}
  {shapeType=Capsule, length=29.9529, radius=2.7222, shapeName=head_capsule, center=<9.3783 -5.2054 -23.4749>, rotation=<rot 0 -0.0012 90.2764>}
 ]
 lowerarm_l []
 lowerarm_r []
 neck_02 []
 spine_01 [{shapeType=Capsule, length=10.8784, radius=17.1628, center=<-8.9021 -0.8973 0>}]
 spine_02 [
  {shapeType=Capsule, length=7.9159, radius=10.7474, center=<2.2663 0.2596 0>}
  {shapeType=Capsule, length=20.6875, radius=9.3125, center=<-1.8317 1.9362 0>, rotation=<rot 90 0 0>}
 ]
 spine_03 [{shapeType=Capsule, length=20.7188, radius=9.2812, center=<-1.2163 0.549 0>, rotation=<rot 90 180 -175.0732>}]
 spine_04 [{shapeType=Capsule, length=10.0312, radius=9.2812, center=<-1.2326 1.308 0>, rotation=<rot 90 0 0>}]
 spine_05 [
  {
   shapeType=Capsule
   length=17.7966
   radius=8.0427
   shapeName=spine_05_capsule
   center=<0.8673 -22.7361 0.2424>
   rotation=<rot -89.1143 179.9994 -175.5266>
  }
  {
   shapeType=Box
   extentX=37.663
   extentY=28.5712
   extentZ=37.133
   shapeName=spine_05_box
   center=<0.0404 3.0487 0.0494>
   rotation=<rot -0.0444 -86.2874 1.6513>
  }
  {
   shapeType=Box
   extentX=10.5346
   extentY=6.643
   extentZ=28.6636
   shapeName=spine_05_box
   center=<13.3574 -19.1051 0.0122>
   rotation=<rot 0 3.5765 0>
  }
 ]
 thigh_l []
 thigh_r []
 upperarm_l []
 upperarm_r []
}
constraints [
 {
  bone1Name=spine_02
  bone2Name=spine_01
  swing1LimitDegrees=45
  swing1Motion=Free
  swing2LimitDegrees=45
  swing2Motion=Free
  twistLimitDegrees=45
  twistMotion=Free
 }
 {
  bone1Name=spine_03
  bone2Name=spine_02
  swing1LimitDegrees=45
  swing1Motion=Free
  swing2LimitDegrees=45
  swing2Motion=Free
  twistLimitDegrees=45
  twistMotion=Free
 }
 {
  bone1Name=spine_04
  bone2Name=spine_03
  swing1LimitDegrees=45
  swing1Motion=Free
  swing2LimitDegrees=45
  swing2Motion=Free
  twistLimitDegrees=45
  twistMotion=Free
 }
 {
  bone1Name=clavicle_l
  bone2Name=spine_05
  swing1LimitDegrees=45
  swing1Motion=Free
  swing2LimitDegrees=45
  swing2Motion=Free
  twistLimitDegrees=45
  twistMotion=Free
 }
 {
  bone1Name=upperarm_l
  bone2Name=clavicle_l
  swing1LimitDegrees=45
  swing1Motion=Free
  swing2LimitDegrees=45
  swing2Motion=Free
  twistLimitDegrees=45
  twistMotion=Free
 }
 {
  bone1Name=lowerarm_l
  bone2Name=upperarm_l
  swing1LimitDegrees=45
  swing1Motion=Free
  swing2LimitDegrees=45
  swing2Motion=Free
  twistLimitDegrees=45
  twistMotion=Free
 }
 {
  bone1Name=neck_02
  bone2Name=spine_05
  swing1LimitDegrees=45
  swing1Motion=Free
  swing2LimitDegrees=45
  swing2Motion=Free
  twistLimitDegrees=45
  twistMotion=Free
 }
 {
  bone1Name=head
  bone2Name=neck_02
  swing1LimitDegrees=45
  swing1Motion=Free
  swing2LimitDegrees=45
  swing2Motion=Free
  twistLimitDegrees=45
  twistMotion=Free
 }
 {
  bone1Name=calf_l
  bone2Name=thigh_l
  swing1LimitDegrees=45
  swing1Motion=Free
  swing2LimitDegrees=45
  swing2Motion=Free
  twistLimitDegrees=45
  twistMotion=Free
 }
 {
  bone1Name=foot_l
  bone2Name=calf_l
  swing1LimitDegrees=45
  swing1Motion=Free
  swing2LimitDegrees=45
  swing2Motion=Free
  twistLimitDegrees=45
  twistMotion=Free
 }
 {
  bone1Name=lowerarm_r
  bone2Name=upperarm_r
  swing1LimitDegrees=45
  swing1Motion=Free
  swing2LimitDegrees=45
  swing2Motion=Free
  twistLimitDegrees=45
  twistMotion=Free
 }
 {
  bone1Name=upperarm_r
  bone2Name=clavicle_r
  swing1LimitDegrees=45
  swing1Motion=Free
  swing2LimitDegrees=45
  swing2Motion=Free
  twistLimitDegrees=45
  twistMotion=Free
 }
 {
  bone1Name=clavicle_r
  bone2Name=spine_05
  swing1LimitDegrees=45
  swing1Motion=Free
  swing2LimitDegrees=45
  swing2Motion=Free
  twistLimitDegrees=45
  twistMotion=Free
 }
 {
  bone1Name=foot_r
  bone2Name=calf_r
  swing1LimitDegrees=45
  swing1Motion=Free
  swing2LimitDegrees=45
  swing2Motion=Free
  twistLimitDegrees=45
  twistMotion=Free
 }
 {
  bone1Name=spine_05
  bone2Name=spine_04
  swing1LimitDegrees=45
  swing1Motion=Free
  swing2LimitDegrees=45
  swing2Motion=Free
  twistLimitDegrees=45
  twistMotion=Free
 }
 {
  bone1Name=calf_r
  bone2Name=thigh_r
  swing1LimitDegrees=45
  swing1Motion=Free
  swing2LimitDegrees=45
  swing2Motion=Free
  twistLimitDegrees=45
  twistMotion=Free
 }
]
